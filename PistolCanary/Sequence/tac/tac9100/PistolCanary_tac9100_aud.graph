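# Audio Level Sequence for the tac9100 sequence, apparently the sound layer for a tentacle boss moment.
# A single master audio track, TentacleAudio, holds two sections over the same short frame range, with volume and pitch at default and no sound assets shown.
# A TentacleBoss possessable of class BGA_Tentacle is bound but carries no tracks, and the sections' attach-actor channels show no target.
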LevelSequence /PistolCanary/Sequence/tac/tac9100/PistolCanary_tac9100_aud
bindings [TentacleBoss {id=#1, possessed_class=BGA_Tentacle_C, tracks []}]
display_rate {denominator=1, numerator=30}
tick_resolution {denominator=1, numerator=24000}
tracks [{
  type=MovieSceneAudioTrack
  display_name=TentacleAudio
  sections [
   {
    type=MovieSceneAudioSection
    end_frame=30
    start_frame=0
    channels [
     Volume {type=MovieSceneScriptingFloatChannel, default=1}
     Pitch {type=MovieSceneScriptingFloatChannel, default=1}
     Master VolumeFMovieSceneFloatChannel {type=MovieSceneScriptingFloatChannel, default=1}
     AttachActor {type=MovieSceneScriptingActorReferenceChannel}
    ]
   }
   {
    type=MovieSceneAudioSection
    end_frame=30
    start_frame=0
    channels [
     Volume {type=MovieSceneScriptingFloatChannel, default=1}
     Pitch {type=MovieSceneScriptingFloatChannel, default=1}
     Master VolumeFMovieSceneFloatChannel {type=MovieSceneScriptingFloatChannel, default=1}
     AttachActor {type=MovieSceneScriptingActorReferenceChannel}
    ]
   }
  ]
 }]
